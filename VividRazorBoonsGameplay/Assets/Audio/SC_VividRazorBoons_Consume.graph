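# Sound cue for consuming a Vivid Razor boon, judging by its name, built from takedown medallion sounds.
# A mixer layers the default medallion pickup sound, slightly lowered, with a boon-specific medallion takedown sound that starts after a short fixed delay.
# A root modulator adds slight random pitch variation to the combined result.
# It uses weapon pickup attenuation and the loot collect sound class, with a concurrency override limiting it to one playing instance.
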SoundCue /VividRazorBoonsGameplay/Assets/Audio/SC_VividRazorBoons_Consume
root=Modulator_0
nodes [
 {id=Delay_0, children [WavePlayer_3], properties {delayMax=0.1, delayMin=0.1}}
 {id=Mixer_0, children [WavePlayer_2,Delay_0], properties {inputVolume [0.8,1]}}
 {id=Modulator_0, children [Mixer_0], properties {pitchMax=1.03, pitchMin=0.97}}
 {id=WavePlayer_2, properties {soundWaveAssetPtr=&/Default_TakedownMedallions_Pickup_SFX}}
 {id=WavePlayer_3, properties {soundWaveAssetPtr=&/VividRazonBoons_TakeDown_Medallion01}}
]
properties {
 attenuationSettings=&Fort_Weapon_Pickup_Att
 bOverrideConcurrency=true
 maxDistance=3500
 soundClassObject=&Fort_Loot_Collect
 volumeMultiplier=0.9
 concurrencyOverrides {maxCount=1}
}
